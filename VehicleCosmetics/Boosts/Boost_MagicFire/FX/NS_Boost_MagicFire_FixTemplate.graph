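# Niagara system for the Magic Fire vehicle boost, trailing colored ribbons and shrinking, fading sprites backward from boost sockets on the car's skeletal mesh.
# Filtered socket count sets ribbon count and sprite partitioning, with DTCarOffset shifting spawn positions.
# Two wide triple-renderer and two narrow ribbon emitters form pairs differing by boneflow offset, all inheriting source movement.
# Each emitter has its own painted-variant palette; frontend preview only changes narrow-ribbon velocity.
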NiagaraSystem /VehicleCosmetics/Boosts/Boost_MagicFire/FX/NS_Boost_MagicFire_FixTemplate
dependencies {
 usedDataInterfaces [&NiagaraDataInterfaceCurve, &NiagaraDataInterfaceColorCurve, &NiagaraDataInterfaceSkeletalMesh]
 usedDynamicInputs []
 usedModules [
  &self:SystemSpawnScript.SystemScriptSource.SystemScriptGraph.NiagaraNodeAssignment_0.SetVariables_#1
  &SystemState
 ]
}
emitters {
 NE_VC_Boost_Ribbon_Parent {
  emitterUpdateScript [
   &EmitterState()
   &VC_FilteredSocketCount()
   &VC_RateQualityHandler()
   &VC_SpawnRibbon_v1(Number of Ribbons=$Emitter.PortCount, SpawnRate=$Emitter.Spawn Rate)
  ]
  particleSpawnScript [
   &VC_PaintedVariantsOverride(
    Black=<rgba 0.1562 0.1562 0.1562 1>
    BurntSienna=<rgba 0.1354 0.0381 0.012 1>
    Cobalt=<rgba 0.5332 0.718 10 1>
    Crimson=<rgba 5 0.3125 0.4396 1>
    DefaultColor=<rgba 0.1 10 5 1>
    ForestGreen=<rgba 1.0037 4.8438 0.8144 1>
    Grey=<rgba 0.8073 0.826 1 1>
    Lime=<rgba 2.6757 6.4531 0.7561 1>
    Orange=<rgba 8 0.963 0 1>
    Pink=<rgba 7.9528 2.0445 10 1>
    Purple=<rgba 2.0944 0.1094 7 1>
    R1 Black:LinearColor=<rgba 0 0 0 1>
    R1 BurntSienna:LinearColor=<rgba 0 0 0 1>
    R1 Cobalt:LinearColor=<rgba 0 0 0 1>
    R1 Crimson:LinearColor=<rgba 0 0 0 1>
    R1 DefaultColor=<rgba 0 0 0 1>
    R1 ForestGreen:LinearColor=<rgba 0 0 0 1>
    R1 Gold:LinearColor=<rgba 0 0 0 1>
    R1 Grey:LinearColor=<rgba 0 0 0 1>
    R1 Lime:LinearColor=<rgba 0 0 0 1>
    R1 Orange:LinearColor=<rgba 0 0 0 1>
    R1 Pink:LinearColor=<rgba 0 0 0 1>
    R1 Purple:LinearColor=<rgba 0 0 0 1>
    R1 Saffron:LinearColor=<rgba 0 0 0 1>
    R1 SkyBlue:LinearColor=<rgba 0 0 0 1>
    R1 TitaniumWhite:LinearColor=<rgba 0 0 0 1>
    R2 DefaultColor:LinearColor=<rgba 0.0075 0.75 0.375 1>
    R2 Gold:LinearColor=<rgba 0.0208 0.0208 0.0208 1>
    Saffron=<rgba 7 5.5538 0.8021 1>
    SkyBlue=<rgba 1.5804 5.1043 20 1>
    TitaniumWhite=<rgba 4.5506 4.2929 5 1>
    Use R1 Painted Overrides=true
    Use R2 Painted Overrides=true
   )
   &VC_InitializeRibbon_v1(
    Color=$Particles.Color
    Custom Offset=true
    Enable Skeletal Position=true
    Lifetime=&MakeCustomFloatFromBool(Bool=$User.bIsFrontendPreview, True Float=0.65, False Float=0.65)
    Local Vector Offset=&AddVectorToPosition(Position=$User.DTCarOffset, Vector=<0 0 0>)
    Number of Bones Per Boneflow group=0
    Offset Amount=1
    Ribbon Width=75
   )
   &AddVelocity(Velocity=&MakeCustomVectorFromBool(Bool=$User.bIsFrontendPreview, True Vector=<-1500 0 0>, False Vector=<-1500 0 0>)) {module=AddVelocity001}
  ]
  particleUpdateScript [
   &ParticleState()
   &InheritSourceMovement(ApplyForce=false, ApplyPosition=true, ApplyPositionFalloff=false)
   &ScaleRibbonWidth(Ribbon Width Scale=&FloatFromCurve(FloatCurve=DataInterface() {propertyValues {Curve {keys [<at 0 1>,<at 1 -0.25>]}}}))
   &SolveForcesAndVelocity(Clamp Velocity=false, Limit Acceleration=false)
   &VC_UpdateRibbon_v1(Enable Skeletal Position=true)
   /SetVariables(Particles.R2 Color:LinearColor=&MultiplyLinearColor(
     Linear Color=$Particles.R2 Color
     Scale Factor=&Vector4FromCurve(Vector4Curve=DataInterface() {propertyValues {
        AlphaCurve {keys [<at 0 1>,<at 0.1 1>,<at 0.3 1>,<at 0.85 0.5>]}
        BlueCurve {keys [<at 0 1>,<at 1 1>]}
        GreenCurve {keys [<at 0 1>,<at 1 1>]}
        RedCurve {keys [<at 0 1>,<at 1 1>]}
       }})
    )) {isSetParameters=true}
  ]
  rendererClasses [&NiagaraRibbonRendererProperties]
  renderers [
   {rendererClass=&NiagaraRibbonRendererProperties}
   {rendererClass=&NiagaraRibbonRendererProperties, rendererIndex=1}
   {rendererClass=&NiagaraRibbonRendererProperties, rendererIndex=2}
  ]
 }
 NE_VC_Boost_Ribbon_Parent001 {
  emitterUpdateScript [
   &EmitterState()
   &VC_FilteredSocketCount()
   &VC_RateQualityHandler()
   &VC_SpawnRibbon_v1(Number of Ribbons=$Emitter.PortCount, SpawnRate=$Emitter.Spawn Rate)
  ]
  particleSpawnScript [
   &VC_PaintedVariantsOverride(
    Black=<rgba 0.099 0.099 0.099 1>
    BurntSienna=<rgba 0.0729 0.0225 0.009 1>
    Cobalt=<rgba 0.0781 0.1692 5 1>
    Crimson=<rgba 3 0.1562 0.2333 1>
    DefaultColor=<rgba 0 15 7.4243 1>
    ForestGreen=<rgba 0.1146 2 0.2098 1>
    Grey=<rgba 0.3906 0.3906 0.3906 1>
    Lime=<rgba 1.5 3 0 1>
    Orange=<rgba 2.0312 0.5261 0.0529 1>
    Pink=<rgba 3 0.3594 2.0398 1>
    Purple=<rgba 0.8331 0.0208 2 1>
    Saffron=<rgba 2 1.6378 0.4479 1>
    SkyBlue=<rgba 1.5104 3.8323 5 1>
   )
   &VC_InitializeRibbon_v1(
    Color=$Particles.Color
    Custom Offset=true
    Enable Skeletal Position=true
    Lifetime=0.65
    Local Vector Offset=&AddVectorToPosition(Position=$User.DTCarOffset, Vector=<0 0 0>)
    Number of Bones Per Boneflow group=0
    Offset Amount=1
    Ribbon Width=50
   )
   &AddVelocity(Velocity=&MakeCustomVectorFromBool(Bool=$User.bIsFrontendPreview, True Vector=<-1200 0 0>, False Vector=<-1500 0 0>)) {module=AddVelocity001}
  ]
  particleUpdateScript [
   &InheritSourceMovement(
    Applied Position Scale=&Multiply_VectorByFloat(
     Vector=<1 1 1>
     Float=&FloatFromCurve(FloatCurve=DataInterface() {propertyValues {Curve {keys [
          {interpMode=RCIM_Cubic, leaveTangent=-0.0051, tangentMode=RCTM_Break, time=0, value=1}
          {arriveTangent=-0.247, interpMode=RCIM_Cubic, tangentMode=RCTM_Break, time=1, value=0.9}
         ]}}})
    )
    ApplyForce=false
    ApplyPosition=true
    ApplyPositionFalloff=false
   )
   &ParticleState()
   &ScaleRibbonWidth(Ribbon Width Scale=&FloatFromCurve(FloatCurve=DataInterface() {propertyValues {Curve {keys [<at 0 1>,<at 0.5 0>]}}}))
   &SolveForcesAndVelocity(Clamp Velocity=false, Limit Acceleration=false)
   &VC_UpdateRibbon_v1(Enable Skeletal Position=true)
  ]
  rendererClasses [&NiagaraRibbonRendererProperties]
  renderers [{rendererClass=&NiagaraRibbonRendererProperties}]
 }
 NE_VC_Boost_Ribbon_Parent002 {
  emitterUpdateScript [
   &EmitterState()
   &VC_FilteredSocketCount()
   &VC_RateQualityHandler(High Spawn Rate=60, Low Spawn Rate=60, Medium Spawn Rate=60, Use Custom Rate?=true)
   &VC_SpawnRibbon_v1(Number of Ribbons=$Emitter.PortCount, SpawnRate=$Emitter.Spawn Rate)
  ]
  particleSpawnScript [
   &VC_PaintedVariantsOverride(
    Black=<rgba 0.1562 0.1562 0.1562 1>
    BurntSienna=<rgba 0.1354 0.0381 0.012 1>
    Cobalt=<rgba 0.5332 0.718 10 1>
    Crimson=<rgba 5 0.3125 0.4396 1>
    DefaultColor=<rgba 0.1 10 5 1>
    ForestGreen=<rgba 1.0037 4.8438 0.8144 1>
    Grey=<rgba 0.8073 0.826 1 1>
    Lime=<rgba 2.6757 6.4531 0.7561 1>
    Orange=<rgba 8 0.963 0 1>
    Pink=<rgba 7.9528 2.0445 10 1>
    Purple=<rgba 2.0944 0.1094 7 1>
    R1 Black:LinearColor=<rgba 0 0 0 1>
    R1 BurntSienna:LinearColor=<rgba 0 0 0 1>
    R1 Cobalt:LinearColor=<rgba 0 0.0125 0.0521 1>
    R1 Crimson:LinearColor=<rgba 0 0 0 1>
    R1 DefaultColor=<rgba 0 0 0 1>
    R1 ForestGreen:LinearColor=<rgba 0 0 0 1>
    R1 Gold:LinearColor=<rgba 0 0 0 1>
    R1 Grey:LinearColor=<rgba 0 0 0 1>
    R1 Lime:LinearColor=<rgba 0 0 0 1>
    R1 Orange:LinearColor=<rgba 0 0 0 1>
    R1 Pink:LinearColor=<rgba 0 0 0 1>
    R1 Purple:LinearColor=<rgba 0 0 0 1>
    R1 Saffron:LinearColor=<rgba 0 0 0 1>
    R1 SkyBlue:LinearColor=<rgba 0 0 0 1>
    R1 TitaniumWhite:LinearColor=<rgba 0 0 0 1>
    R2 Black:LinearColor=<rgba 0.0469 0.0469 0.0469 1>
    R2 BurntSienna:LinearColor=<rgba 0.0729 0.0205 0.0065 1>
    R2 Cobalt:LinearColor=<rgba 0.0533 0.0718 1 1>
    R2 Crimson:LinearColor=<rgba 0.7913 0.0495 0.0696 1>
    R2 DefaultColor:LinearColor=<rgba 0.0075 0.75 0.375 1>
    R2 ForestGreen:LinearColor=<rgba 0.0891 0.6 0.1189 1>
    R2 Gold:LinearColor=<rgba 0.0208 0.0208 0.0208 1>
    R2 Grey:LinearColor=<rgba 0.1598 0.1635 0.1979 1>
    R2 Lime:LinearColor=<rgba 0.2488 0.6 0.0703 1>
    R2 Orange:LinearColor=<rgba 0.7 0.0843 0 1>
    R2 Pink:LinearColor=<rgba 0.5965 0.1533 0.75 1>
    R2 Purple:LinearColor=<rgba 0.3601 0 1.25 1>
    R2 Saffron:LinearColor=<rgba 1 0.7934 0.1146 1>
    R2 SkyBlue:LinearColor=<rgba 0.079 0.2552 1 1>
    Saffron=<rgba 7 5.5538 0.8021 1>
    SkyBlue=<rgba 1.5804 5.1043 20 1>
    TitaniumWhite=<rgba 4.5506 4.2929 5 1>
    Use R1 Painted Overrides=true
    Use R2 Painted Overrides=true
   )
   &VC_InitializeRibbon_v1(
    Boneflow Offset=1
    Color=$Particles.Color
    Custom Offset=true
    Enable Skeletal Position=true
    Lifetime=&MakeCustomFloatFromBool(Bool=$User.bIsFrontendPreview, True Float=0.65, False Float=0.65)
    Local Vector Offset=&AddVectorToPosition(Position=$User.DTCarOffset, Vector=<0 0 0>)
    Number of Bones Per Boneflow group=0
    Offset Amount=1
    Ribbon Width=75
   )
   &AddVelocity(Velocity=&MakeCustomVectorFromBool(Bool=$User.bIsFrontendPreview, True Vector=<-1500 0 0>, False Vector=<-1500 0 0>)) {module=AddVelocity001}
  ]
  particleUpdateScript [
   &ParticleState()
   &InheritSourceMovement(ApplyForce=false, ApplyPosition=true, ApplyPositionFalloff=false)
   &ScaleRibbonWidth(Ribbon Width Scale=&FloatFromCurve(FloatCurve=DataInterface() {propertyValues {Curve {keys [<at 0 1>,<at 1 -0.25>]}}}))
   &SolveForcesAndVelocity(Clamp Velocity=false, Limit Acceleration=false)
   &VC_UpdateRibbon_v1(Enable Skeletal Position=true)
   /SetVariables(Particles.R2 Color:LinearColor=&MultiplyLinearColor(
     Linear Color=$Particles.R2 Color
     Scale Factor=&Vector4FromCurve(Vector4Curve=DataInterface() {propertyValues {
        AlphaCurve {keys [<at 0 1>,<at 0.1 1>,<at 0.3 1>,<at 0.85 0.5>]}
        BlueCurve {keys [<at 0 1>,<at 1 1>]}
        GreenCurve {keys [<at 0 1>,<at 1 1>]}
        RedCurve {keys [<at 0 1>,<at 1 1>]}
       }})
    )) {isSetParameters=true}
  ]
  rendererClasses [&NiagaraRibbonRendererProperties]
  renderers [
   {rendererClass=&NiagaraRibbonRendererProperties}
   {rendererClass=&NiagaraRibbonRendererProperties, rendererIndex=1}
   {rendererClass=&NiagaraRibbonRendererProperties, rendererIndex=2}
  ]
 }
 NE_VC_Boost_Ribbon_Parent005 {
  emitterUpdateScript [
   &EmitterState()
   &VC_FilteredSocketCount()
   &VC_RateQualityHandler()
   &VC_SpawnRibbon_v1(Number of Ribbons=$Emitter.PortCount, SpawnRate=$Emitter.Spawn Rate)
  ]
  particleSpawnScript [
   &VC_PaintedVariantsOverride(
    Black=<rgba 0.099 0.099 0.099 1>
    BurntSienna=<rgba 0.0729 0.0225 0.009 1>
    Cobalt=<rgba 0.0781 0.1692 5 1>
    Crimson=<rgba 3 0.1562 0.2333 1>
    DefaultColor=<rgba 0 15 7.4243 1>
    ForestGreen=<rgba 0.1146 2 0.2098 1>
    Grey=<rgba 0.3906 0.3906 0.3906 1>
    Lime=<rgba 1.5 3 0 1>
    Orange=<rgba 2.0312 0.5261 0.0529 1>
    Pink=<rgba 3 0.3594 2.0398 1>
    Purple=<rgba 0.8331 0.0208 2 1>
    Saffron=<rgba 2 1.6378 0.4479 1>
    SkyBlue=<rgba 1.5104 3.8323 5 1>
   )
   &VC_InitializeRibbon_v1(
    Boneflow Offset=1
    Color=$Particles.Color
    Custom Offset=true
    Enable Skeletal Position=true
    Lifetime=0.65
    Local Vector Offset=&AddVectorToPosition(Position=$User.DTCarOffset, Vector=<0 0 0>)
    Number of Bones Per Boneflow group=0
    Offset Amount=1
    Ribbon Width=50
   )
   &AddVelocity(Velocity=&MakeCustomVectorFromBool(Bool=$User.bIsFrontendPreview, True Vector=<-1200 0 0>, False Vector=<-1500 0 0>)) {module=AddVelocity001}
  ]
  particleUpdateScript [
   &InheritSourceMovement(
    Applied Position Scale=&Multiply_VectorByFloat(
     Vector=<1 1 1>
     Float=&FloatFromCurve(FloatCurve=DataInterface() {propertyValues {Curve {keys [
          {interpMode=RCIM_Cubic, leaveTangent=-0.0051, tangentMode=RCTM_Break, time=0, value=1}
          {arriveTangent=-0.247, interpMode=RCIM_Cubic, tangentMode=RCTM_Break, time=1, value=0.9}
         ]}}})
    )
    ApplyForce=false
    ApplyPosition=true
    ApplyPositionFalloff=false
   )
   &ParticleState()
   &ScaleRibbonWidth(Ribbon Width Scale=&FloatFromCurve(FloatCurve=DataInterface() {propertyValues {Curve {keys [<at 0 1>,<at 0.5 0>]}}}))
   &SolveForcesAndVelocity(Clamp Velocity=false, Limit Acceleration=false)
   &VC_UpdateRibbon_v1(Enable Skeletal Position=true)
  ]
  rendererClasses [&NiagaraRibbonRendererProperties]
  renderers [{rendererClass=&NiagaraRibbonRendererProperties}]
 }
 NE_VC_Boost_Sprite_Parent001 {
  emitterUpdateScript [
   &EmitterState()
   &VC_FilteredSocketCount()
   &SpawnRate(SpawnRate=60, Use Spawn Probability=false)
  ]
  particleSpawnScript [
   &VC_PaintedVariantsOverride(
    Black=<rgba 0.5938 0.5938 0.5938 10>
    BurntSienna=<rgba 1.75 0.6262 0.1506 10>
    Cobalt=<rgba 0 1.85 100 10>
    Crimson=<rgba 25 4.1667 4.7313 10>
    DefaultColor=<rgba 1 100 50.0001 2>
    ForestGreen=<rgba 3.5156 25 4.6011 10>
    Grey=<rgba 2 2 2 10>
    Lime=<rgba 5 10 0 10>
    Orange=<rgba 15 5.2321 0 10>
    Pink=<rgba 10 4.0104 7.822 10>
    Purple=<rgba 4.9947 1.5104 10 10>
    Saffron=<rgba 12 9.2 0 10>
    SkyBlue=<rgba 8.3333 16.0962 20 10>
    TitaniumWhite=<rgba 5 5 5 10>
   )
   &InitializeParticle(
    Color=$Particles.Color
    Lifetime=0.75
    Sprite Rotation Mode=::NewEnumerator2
    Sprite Size Mode=::NewEnumerator1
    Uniform Sprite Size Max=60
    Uniform Sprite Size Min=40
    UsePositionOffset=false
    Write Material Random=true
   )
   &PartitionParticles(Number Of Partitions=$Emitter.PortCount)
   &SkeletalMeshLocation(
    Bone / Socket Index=$Particles.PartitionParticles.Partition
    Bone / Socket Sampling Mode=::NewEnumerator5
    MaskSpawnGroup=false
    Mesh Sampling Type=::NewEnumerator6
    Sampled Position Offset=&AddVectorToPosition(Position=$User.DTCarOffset, Vector=<-10 0 0>)
    Skeletal Mesh=$System.SkeletalMesh
   )
   &AddVelocity(Velocity=<-500 0 0>)
   &ShapeLocation(MaskSpawnGroup=false, Override Seed=false, Sphere Radius=25, UseOwnerScale=true)
  ]
  particleUpdateScript [
   &ParticleState()
   &ScaleSpriteSize(Uniform Curve Sprite Scale=DataInterface() {propertyValues {Curve {
       defaultValue=0
       keys [
        {arriveTangent=-0.089, interpMode=RCIM_Cubic, leaveTangent=-0.089, tangentMode=RCTM_User, time=0, value=1}
        {arriveTangent=-0.1517, tangentMode=RCTM_User, time=1, value=0}
       ]
      }}})
   &SolveForcesAndVelocity(Clamp Velocity=false, Limit Acceleration=false)
   &ScaleColor(
    Scale Alpha=&FloatFromCurve(FloatCurve=DataInterface() {propertyValues {Curve {keys [<at 0 1>,<at 0.5 1>,<at 1 0>]}}})
    ScaleA=true
    ScaleRGB=true
   )
  ]
  rendererClasses [&NiagaraSpriteRendererProperties]
  renderers [{rendererClass=&NiagaraSpriteRendererProperties}]
 }
}
summary {
 systemName=NS_Boost_MagicFire_FixTemplate
 emitters [
  {emitterName=NE_VC_Boost_Ribbon_Parent, rendererClasses [&NiagaraRibbonRendererProperties]}
  {emitterName=NE_VC_Boost_Sprite_Parent001, rendererClasses [&NiagaraSpriteRendererProperties]}
  {emitterName=NE_VC_Boost_Ribbon_Parent001, rendererClasses [&NiagaraRibbonRendererProperties]}
  {emitterName=NE_VC_Boost_Ribbon_Parent002, rendererClasses [&NiagaraRibbonRendererProperties]}
  {emitterName=NE_VC_Boost_Ribbon_Parent005, rendererClasses [&NiagaraRibbonRendererProperties]}
 ]
 userVariables [
  User.bIsFrontendPreview {type=NiagaraBool}
  User.bIsPainted {type=NiagaraBool}
  User.ColorVariantID {type=NiagaraInt32}
  User.DTCarOffset {type=NiagaraPosition, description="", defaultValue {}}
  User.PaintedColor {type=LinearColor, description="", defaultValue {struct=LinearColor, value=<rgba 1 1 1 1>}}
  User.PaintedScalar {type=NiagaraFloat, defaultValue=1, description=""}
  User.PortCount {type=NiagaraInt32, defaultValue=1}
  User.SizeScaling {type=NiagaraFloat, defaultValue=5, description=""}
  User.SubColor {type=LinearColor, description="", defaultValue {struct=LinearColor, value=<rgba 1 1 1 1>}}
  User.SubColorCustom {type=LinearColor, description="", defaultValue {struct=LinearColor, value=<rgba 1 1 1 1>}}
 ]
}
